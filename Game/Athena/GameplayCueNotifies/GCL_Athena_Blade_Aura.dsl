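; Looping gameplay cue for the Athena blade aura that toggles its Niagara effects with the player's skydiving and holster state.
; On loop start it caches the spawned Niagara components and binds to the PlayerPawn_Athena skydiving and holster events, unbinding on removal.
; Skydive begin and holster call Unhide, which despite its name hides and deactivates the components; skydive end and unholster call Show, which reveals and reactivates them.
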(event EventOnLoopingStart(Niagara)
  (params (MyTarget :type Actor%20Object%20Reference) (Parameters :type Gameplay%20Cue%20Parameters%20Structure%20%28by%20ref%29) (ParticleComponents :type Array%20of%20Niagara%20Particle%20System%20Component%20Object%20References) (AudioComponents :type Array%20of%20Audio%20Component%20Object%20References))
  (Variables|Default|SetNS_Spawned%20Emitters ParticleComponents)
  (bind _asplayer_pawn_athena (Utilities|Casting|CastToPlayerPawn_Athena MyTarget))
  (EventDispatchers|BindEventtoOnBeginSkydivingEvent _asplayer_pawn_athena (EventDispatchers|CreateEvent :!event Custom|Unhide))
  (EventDispatchers|BindEventtoOnEndSkydivingEvent _asplayer_pawn_athena (EventDispatchers|CreateEvent :!event Custom|Show))
  (EventDispatchers|BindEventtoOnHolsteredEvent _asplayer_pawn_athena (EventDispatchers|CreateEvent :!event Custom|Unhide))
  (EventDispatchers|BindEventtoOnUnholsteredEvent _asplayer_pawn_athena (EventDispatchers|CreateEvent :!event Custom|Show)))

(event EventOnLoopingStart(Cascade)
  (params (MyTarget :type Actor%20Object%20Reference) (Parameters :type Gameplay%20Cue%20Parameters%20Structure%20%28by%20ref%29) (ParticleComponents :type Array%20of%20Cascade%20Particle%20System%20Component%20Object%20References) (AudioComponents :type Array%20of%20Audio%20Component%20Object%20References)))

(event EventOnRemoval(Cascade)
  (params (MyTarget :type Actor%20Object%20Reference) (Parameters :type Gameplay%20Cue%20Parameters%20Structure%20%28by%20ref%29) (ParticleComponents :type Array%20of%20Cascade%20Particle%20System%20Component%20Object%20References) (AudioComponents :type Array%20of%20Audio%20Component%20Object%20References) (BurstCameraShakeInstance :type Camera%20Shake%20Base%20Object%20Reference) (BurstDecalInstance :type Decal%20Actor%20Object%20Reference)))

(event EventOnRemoval(Niagara)
  (params (MyTarget :type Actor%20Object%20Reference) (Parameters :type Gameplay%20Cue%20Parameters%20Structure%20%28by%20ref%29) (ParticleComponents :type Array%20of%20Niagara%20Particle%20System%20Component%20Object%20References) (AudioComponents :type Array%20of%20Audio%20Component%20Object%20References) (BurstCameraShakeInstance :type Camera%20Shake%20Base%20Object%20Reference) (BurstDecalInstance :type Decal%20Actor%20Object%20Reference))
  (bind _asplayer_pawn_athena (Utilities|Casting|CastToPlayerPawn_Athena MyTarget))
  (EventDispatchers|UnbindEventfromOnBeginSkydivingEvent _asplayer_pawn_athena (EventDispatchers|CreateEvent :!event Custom|Unhide))
  (EventDispatchers|UnbindEventfromOnEndSkydivingEvent _asplayer_pawn_athena (EventDispatchers|CreateEvent :!event Custom|Show))
  (EventDispatchers|UnbindEventfromOnHolsteredEvent _asplayer_pawn_athena (EventDispatchers|CreateEvent :!event Custom|Unhide))
  (EventDispatchers|UnbindEventfromOnUnholsteredEvent _asplayer_pawn_athena (EventDispatchers|CreateEvent :!event Custom|Show)))

(event Custom|Show
  (for _array_element (Variables|Default|GetNS_Spawned%20Emitters)
    (Development|SetHiddeninGame _array_element false false)
    (Components|Activation|Activate _array_element true)))

(event Custom|Unhide
  (for _array_element (Variables|Default|GetNS_Spawned%20Emitters)
    (Development|SetHiddeninGame _array_element true false)
    (Components|Activation|Deactivate _array_element)))

(fn ConstructionScript)
