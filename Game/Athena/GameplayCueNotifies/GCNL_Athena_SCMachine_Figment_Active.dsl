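; Looping gameplay cue for the active "Figment" effect on an Athena SCMachine, keeping its particles placed at the machine's holo logo.
; On looping start and on removal, it casts the target to the BPI_BGAAthenaSCMachine interface and fetches the holo logo Niagara component.
; It then moves the first particle component of that Cascade event to the logo's world location.
; BeginPlay, overlap, tick and construction script are empty.
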(event EventBeginPlay)

(event Collision|EventActorBeginOverlap
  (params (OtherActor :type Actor%20Object%20Reference)))

(event EventTick
  (params (DeltaSeconds :type Float%20%28single-precision%29)))

(event EventOnLoopingStart(Cascade)
  (params (MyTarget :type Actor%20Object%20Reference) (Parameters :type Gameplay%20Cue%20Parameters%20Structure%20%28by%20ref%29) (ParticleComponents :type Array%20of%20Cascade%20Particle%20System%20Component%20Object%20References) (AudioComponents :type Array%20of%20Audio%20Component%20Object%20References))
  ;[1 1 1] Move the GC to the correct location
  (bind _asbpi_bgaathena_scmachine (Utilities|Casting|CastToBPI_BGAAthenaSCMachine MyTarget))
  (bind _holologonscomponent (BPI_BGAAthenaSCMachine|GetHoloLogoNiagaraComponent _asbpi_bgaathena_scmachine))
  (Transformation|SetWorldLocation (Utilities|Array|Get(acopy) ParticleComponents) (Transformation|GetWorldLocation _holologonscomponent) false false))

(event EventOnRemoval(Cascade)
  (params (MyTarget :type Actor%20Object%20Reference) (Parameters :type Gameplay%20Cue%20Parameters%20Structure%20%28by%20ref%29) (ParticleComponents :type Array%20of%20Cascade%20Particle%20System%20Component%20Object%20References) (AudioComponents :type Array%20of%20Audio%20Component%20Object%20References) (BurstCameraShakeInstance :type Camera%20Shake%20Base%20Object%20Reference) (BurstDecalInstance :type Decal%20Actor%20Object%20Reference))
  ;[1 1 1] Move the GC to the correct location
  (bind _asbpi_bgaathena_scmachine (Utilities|Casting|CastToBPI_BGAAthenaSCMachine MyTarget))
  (bind _holologonscomponent (BPI_BGAAthenaSCMachine|GetHoloLogoNiagaraComponent _asbpi_bgaathena_scmachine))
  (Transformation|SetWorldLocation (Utilities|Array|Get(acopy) ParticleComponents) (Transformation|GetWorldLocation _holologonscomponent) false false))

(fn ConstructionScript)
