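; Looping gameplay cue that shows paint splatter on an Athena player pawn, with a drip particle effect on its mesh.
; On application it clears prior effects, then attaches absolute-rotation mesh planes with an outline material to configured sockets.
; A timer or cue removal destroys the planes and deactivates the drips.
; A helper packs fade timing, height and local-control data into Niagara vectors, though its caller passes no Niagara component.
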(event EventOnApplication(Cascade)
  (params (MyTarget :type Actor%20Object%20Reference) (Parameters :type Gameplay%20Cue%20Parameters%20Structure%20%28by%20ref%29) (ParticleComponents :type Array%20of%20Cascade%20Particle%20System%20Component%20Object%20References) (AudioComponents :type Array%20of%20Audio%20Component%20Object%20References) (BurstCameraShakeInstance :type Camera%20Shake%20Base%20Object%20Reference) (BurstDecalInstance :type Decal%20Actor%20Object%20Reference))
  (bind _asplayer_pawn_athena (Utilities|Casting|CastToPlayerPawn_Athena MyTarget))
  (Utilities|IsValid _asplayer_pawn_athena
    (:Is%20Valid
      (bind _output_get (Variables|Default|SetPawn _asplayer_pawn_athena))
      (Variables|Default|SetPawnSkeletalMesh (Variables|/Script/Engine.Character|GetMesh _output_get))
      (bind (_origin _boxextent) (Collision|GetComponentBounds (Variables|/Script/Engine.Character|GetMesh _output_get)))
      (bind (_boxextent_x _boxextent_y _boxextent_z) (Math|Vector|BreakVector _boxextent))
      (Variables|Default|SetPlayerHeight (* _boxextent_z 2.0))
      (bind _result (Math|Float|NegateFloat _boxextent_z)
        (:%20%20
          (Variables|Default|SetPlayerHalfHeightNegated _result)
          (CallFunction|RemoveExistingFX)
          (Utilities|Time|ClearandInvalidateTimerbyHandle 0 (Variables|Default|GetRemovalTimerHandle))
          (CallFunction|ApplyDripFX)
          (bind _values (Utilities|Map|Values (Variables|Default|GetBonesAndRelativeTransforms)))
          (Variables|Default|SetArrayValues _values)
          (bind _keys (Utilities|Map|Keys (Variables|Default|GetBonesAndRelativeTransforms)))
          (for (_array_element _array_index) _keys
            (Variables|Default|SetCurrentSocket _array_element)
            (if (Transformation|DoesSocketExist (Variables|Default|GetPawnSkeletalMesh) _array_element)
              (bind _returnvalue (AddComponent|NODE_AddStaticMeshComponent-0))
              (Transformation|AttachComponentToComponent _returnvalue (Variables|Default|GetPawnSkeletalMesh) (Variables|Default|GetCurrentSocket) "KeepRelative" "KeepRelative" "KeepRelative" true)
              (Transformation|SetRelativeTransform _returnvalue (Utilities|Array|Get(aref) (Variables|Default|GetArrayValues) _array_index) false true)
              (Rendering|Material|SetMaterial _returnvalue 0 (Variables|Default|GetOutlineMaterial))
              (Rendering|SetTranslucentSortPriority _returnvalue (Variables|Default|GetPlaneSortPriority))
              (Variables|/Script/Engine.SceneComponent|SetbAbsoluteRotation true _returnvalue)
              (Utilities|Array|AddUnique (Variables|Default|GetPlanes) _returnvalue)
              (CallFunction|SetCustomPrimitiveVariables _returnvalue)))
          (bind _returnvalue_2 (Utilities|Time|SetTimerbyFunctionName 0 "RemoveExistingFX" (Variables|Default|GetEffectDuration)))
          (Variables|Default|SetRemovalTimerHandle _returnvalue_2))))))

(event EventOnRemoval(Cascade)
  (params (MyTarget :type Actor%20Object%20Reference) (Parameters :type Gameplay%20Cue%20Parameters%20Structure%20%28by%20ref%29) (ParticleComponents :type Array%20of%20Cascade%20Particle%20System%20Component%20Object%20References) (AudioComponents :type Array%20of%20Audio%20Component%20Object%20References) (BurstCameraShakeInstance :type Camera%20Shake%20Base%20Object%20Reference) (BurstDecalInstance :type Decal%20Actor%20Object%20Reference))
  (CallFunction|RemoveExistingFX))

(fn ApplyDripFX
  (bind _returnvalue (Effects|SpawnEmitterAttached :EmitterTemplate "/Game/Athena/Items/Weapons/Prototype/Papaya/P_PaintSplat_Drips.P_PaintSplat_Drips" :AttachToComponent (Variables|Default|GetPawnSkeletalMesh) :AttachPointName "None" :Scale "1.000000,1.000000,1.000000" :LocationType "KeepRelativeOffset" :bAutoDestroy true :PoolingMethod "AutoRelease" :bAutoActivate true))
  (bind _output_get (Variables|Default|SetDripFX _returnvalue)))

(fn OnExecute
  (params (MyTarget :type Actor%20Object%20Reference) (Parameters :type Gameplay%20Cue%20Parameters%20Structure%20%28by%20ref%29))
  (return false))

(fn RemoveExistingFX
  (for _array_element (Variables|Default|GetPlanes)
    (Utilities|IsValid _array_element
      (:Is%20Valid
        (Components|DestroyComponent _array_element 0))))
  (Utilities|Array|Clear (Variables|Default|GetPlanes))
  (bind _dripfx (Variables|Default|GetDripFX)
    (:then
      (Components|Activation|Deactivate _dripfx))
    (:else)))

(fn SetCustomPrimitiveVariables
  (params (InputPin :type Static%20Mesh%20Component%20Object%20Reference) (Niagara%20FX%20Mud%20Cover :type Niagara%20Particle%20System%20Component%20Object%20Reference))
  ;[0.2465 0.5337 1] Send Parameters to Niagara Mud FX. for body Decal control.
  (Niagara|SetNiagaraVariable(Vector4) Niagara%20FX%20Mud%20Cover "InputFromBP1" (Math|Vector4|MakeVector4 (Utilities|Time|GetGameTimeinSeconds 0) (Variables|Default|GetTransitionIn) (+ (Utilities|Time|GetGameTimeinSeconds 0) (Variables|Default|GetDelayBeforeFade)) (Variables|Default|GetTransitionOut)))
  (Niagara|SetNiagaraVariable(Vector4) Niagara%20FX%20Mud%20Cover "InputFromBP2" (Math|Vector4|MakeVector4 (Variables|Default|GetPlayerHalfHeightNegated) (+ (Variables|Default|GetPlayerHeight) (Variables|Default|GetPlayerHeightPadding)) (Math|Conversions|ToFloat(Boolean) (Pawn|IsLocallyControlled (Variables|Default|GetPawn))))))

(fn ConstructionScript)
